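# Sound cue for the spin sound of the Saucer emote.
# A Random node picks one of five equally weighted spin waves and feeds a Modulator that varies pitch on playback.
# It uses emote attenuation, the emote music sound class and an emote music concurrency set, and also sends to a music audio analysis submix.
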SoundCue /BRCosmetics/Athena/Sounds/Emotes/Saucer/Emote_Saucer_Spin_Cue
root=Modulator_0
nodes [
 {id=Modulator_0, children [Random_0], properties {pitchMin=0.9}}
 {id=Random_0, children [WavePlayer_2,WavePlayer_3,WavePlayer_4,WavePlayer_5,WavePlayer_6], properties {weights [1,1,1,1,1]}}
 {id=WavePlayer_2, properties {soundWaveAssetPtr=/Game/Athena/Sounds/Emotes/Saucer/Emote_Saucer_Spin_01}}
 {id=WavePlayer_3, properties {soundWaveAssetPtr=/Game/Athena/Sounds/Emotes/Saucer/Emote_Saucer_Spin_02}}
 {id=WavePlayer_4, properties {soundWaveAssetPtr=/Game/Athena/Sounds/Emotes/Saucer/Emote_Saucer_Spin_03}}
 {id=WavePlayer_5, properties {soundWaveAssetPtr=/Game/Athena/Sounds/Emotes/Saucer/Emote_Saucer_Spin_04}}
 {id=WavePlayer_6, properties {soundWaveAssetPtr=/Game/Athena/Sounds/Emotes/Saucer/Emote_Saucer_Spin_05}}
]
properties {
 attenuationSettings=&Fort_Emotes_Athena_Att
 maxDistance=1600
 soundClassObject=&Fort_Music_Emote
 volumeMultiplier=1
 concurrencyOverrides {maxCount=3}
 concurrencySet [/Game/Sounds/Concurrency_Rules/Athena_EmoteMusic_Conc]
 soundSubmixSends [{minSendLevel=0, soundSubmix=&Fortnite_AudioAnalysis_Music}]
}
