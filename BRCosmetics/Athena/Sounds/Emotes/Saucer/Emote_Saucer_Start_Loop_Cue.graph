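# Sound cue for the Saucer emote that layers a one-shot start sound with a sustained loop.
# A Mixer plays both branches together, with the looping wave briefly delayed and faded in by a volume envelope.
# A continuous modulator maps the emote move speed parameter to the loop's pitch and volume.
# It uses emote music class, attenuation and concurrency, and sends to a music audio-analysis submix.
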SoundCue /BRCosmetics/Athena/Sounds/Emotes/Saucer/Emote_Saucer_Start_Loop_Cue
root=Mixer_0
nodes [
 {id=Delay_0, children [WavePlayer_0], properties {delayMax=0.5, delayMin=0.5}}
 {
  id=Enveloper_0
  children [Delay_0]
  properties {pitchCurve {editorCurveData {keys [{time=0}]}}, volumeCurve {editorCurveData {keys [<at 0 0>,<at 1 1>]}}}
 }
 {id=Mixer_0, children [ModulatorContinuous_0, WavePlayer_1], properties {inputVolume [1,1]}}
 {
  id=ModulatorContinuous_0
  children [Enveloper_0]
  properties {
   pitchModulationParams {maxOutput=1.1, minInput=0.1, minOutput=0.9, parameterName=EmoteMoveSpeed}
   volumeModulationParams {minInput=0.5, minOutput=0.7, parameterName=emoteMoveSpeed}
  }
 }
 {id=WavePlayer_0, properties {bLooping=true, soundWaveAssetPtr=/Game/Athena/Sounds/Emotes/Saucer/Emote_Saucer_Loop_01}}
 {id=WavePlayer_1, properties {soundWaveAssetPtr=/Game/Athena/Sounds/Emotes/Saucer/Emote_Saucer_Start_01}}
]
properties {
 attenuationSettings=&Fort_Emotes_Athena_Att
 maxDistance=1600
 soundClassObject=&Fort_Music_Emote
 volumeMultiplier=0.7
 concurrencyOverrides {maxCount=3}
 concurrencySet [/Game/Sounds/Concurrency_Rules/Athena_EmoteMusic_Conc]
 soundSubmixSends [{minSendLevel=0, soundSubmix=&Fortnite_AudioAnalysis_Music}]
}
